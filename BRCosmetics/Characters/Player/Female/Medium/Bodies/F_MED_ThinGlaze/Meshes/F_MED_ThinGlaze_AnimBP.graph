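# Character-part AnimBlueprint for the F_MED_ThinGlaze female body, adding secondary motion to clothing and accessories.
# It copies the parent mesh pose into a Control Rig, then chains RigidBody physics, button and zipper constraints, and per-bone AnimDynamics for zipper, flashlight, pepper spray, radio and donuts, using cached poses.
# The EventGraph sets gravity override from a parameter asset, a front-end value, and a vector toggled by zipline enter/exit events.
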AnimBlueprint /BRCosmetics/Characters/Player/Female/Medium/Bodies/F_MED_ThinGlaze/Meshes/F_MED_ThinGlaze_AnimBP
target_skeleton=&/F_MED_ThinGlaze_Skeleton
graphs [
 AnimGraph {nodes [
   Output Pose(Result=Use cached pose 'dyn_cache'()) {layerGroup=None}
   n1 = Control Rig!(Source=Copy Pose From Mesh() {bUseAttachedParent=true}, Alpha, IsFrontend:bool=Get bIsInFrontEnd()) -> (Pose) {controlRigAssetReference {blueprintRigClass=&/F_MED_ThinGlaze_CtrlRig.F_MED_ThinGlaze_CtrlRig_C}}
   n2 = RigidBody!(ComponentPose=Local To Component(LocalPose=@n1), Alpha, OverrideWorldGravity:Vector="0,0,-980") -> (Pose) {
    bClampLinearTranslationLimitToRefPose=true
    bOverrideWorldGravity=true
    overridePhysicsAsset=&/F_MED_ThinGlaze_Physics_Gun
    simulationSpace=BaseBoneSpace
    alphaBoolBlend {blendOption=Linear}
    baseBoneRef {boneName=pelvis}
    overrideWorldGravity {z=-980}
    simSpaceSettings {maxAngularAcceleration=40, worldAlpha=0.8, externalLinearDragV=<0.1 0.1 0.1>}
   }
   Save cached pose 'control_rig_cache'(Pose=Component To Local(ComponentPose=@n2)) {cacheName=control_rig_cache}
   n3 = "Constraint - base_button_front_lf (PPP)"!(
    ComponentPose=Local To Component(LocalPose=Use cached pose 'control_rig_cache'())
    Alpha=1
    ConstraintWeights_0=0.7626
    ConstraintWeights_1:float=0.237
    ConstraintWeights_2:float=0.0004
   ) -> (Pose) {
    boneToModify {boneName=base_button_front_lf}
    constraintSetup [{targetBone {boneName=spine_05}}, {targetBone {boneName=spine_04}}, {targetBone {boneName=clavicle_l}}]
    constraintWeights [1,1,1]
   }
   n4 = "Constraint - base_zipper_front_lf (PPPP)"!(
    ComponentPose=@n3
    Alpha=1
    ConstraintWeights_0=0.8934
    ConstraintWeights_1:float=0.1048
    ConstraintWeights_2:float=0.0009
    ConstraintWeights_3:float=0.0008
   ) -> (Pose) {
    boneToModify {boneName=base_zipper_front_lf}
    constraintSetup [{targetBone {boneName=spine_05}}, {targetBone {boneName=spine_04}}, {targetBone {boneName=clavicle_r}}, {targetBone {boneName=clavicle_l}}]
    constraintWeights [1,1,1,1]
   }
   Save cached pose 'predyn_cache'(Pose=Component To Local(ComponentPose=@n4)) {cacheName=predyn_cache}
   n5 = Anim Dynamics - Bone: dyn_zipper_front_lf!(
    ComponentPose=Local To Component(LocalPose=Use cached pose 'predyn_cache'())
    Alpha=1
    GravityOverride:Vector="0,0,-980"
    ExternalForce="0,0,0"
   ) -> (Pose) {
    angularDampingOverride=1
    angularSpringConstant=700
    bOverrideAngularDamping=true
    bUseGravityOverride=true
    gravityScale=0.35
    simulationSpace=RootRelative
    alphaBoolBlend {blendOption=Linear}
    alphaScaleBiasClamp {bInterpResult=true, interpSpeedDecreasing=2.5, interpSpeedIncreasing=8}
    boundBone {boneName=dyn_zipper_front_lf}
    componentAppliedLinearAccClamp=<3000 3000 3000>
    componentLinearAccScale=<0.3 0.3 0.3>
    componentLinearVelScale=<0.3 0.3 0.3>
    gravityOverride {z=-980}
    physicsBodyDefinitions [{
      boundBone {boneName=dyn_zipper_front_lf}
      boxExtents=<5 8 5>
      constraintSetup {angularLimitsMax=<0 10 15>, angularLimitsMin=<0 -10 0>, angularTarget=<1 0 0>}
      localJointOffset=<1 0 0>
     }]
    relativeSpaceBone {boneName=root}
    simSpaceSettings {maxAngularAcceleration=35, maxAngularVelocity=7, simSpaceAngularAlpha=1}
   }
   n6 = Anim Dynamics - Bone: dyn_flashlight_lf!(ComponentPose=@n5, Alpha=1, GravityOverride:Vector="0,0,-980", ExternalForce="0,0,0") -> (Pose) {
    angularDampingOverride=1
    angularSpringConstant=150
    bAngularSpring=true
    bOverrideAngularDamping=true
    bUseGravityOverride=true
    gravityScale=0.35
    simulationSpace=RootRelative
    boundBone {boneName=dyn_flashlight_lf}
    componentAppliedLinearAccClamp=<3000 3000 3000>
    componentLinearAccScale=<0.3 0.3 0.3>
    componentLinearVelScale=<0.3 0.3 0.3>
    gravityOverride {z=-980}
    physicsBodyDefinitions [{
      boundBone {boneName=dyn_flashlight_lf}
      boxExtents=<20 30 30>
      constraintSetup {angularLimitsMax=<5 0 5>, angularLimitsMin=<-3 0 -5>, angularTarget=<1 0 0>}
      localJointOffset=<2 0 0>
     }]
    simSpaceSettings {maxAngularAcceleration=35, maxAngularVelocity=7, simSpaceAngularAlpha=1}
   }
   n7 = Anim Dynamics - Bone: dyn_pepperspray_rt!(ComponentPose=@n6, Alpha=1, GravityOverride:Vector="0,0,-980", ExternalForce="0,0,0") -> (Pose) {
    angularDampingOverride=1
    angularSpringConstant=150
    bAngularSpring=true
    bOverrideAngularDamping=true
    bUseGravityOverride=true
    gravityScale=0.35
    simulationSpace=RootRelative
    boundBone {boneName=dyn_pepperspray_rt}
    componentAppliedLinearAccClamp=<3000 3000 3000>
    componentLinearAccScale=<0.3 0.3 0.3>
    componentLinearVelScale=<0.3 0.3 0.3>
    gravityOverride {z=-980}
    physicsBodyDefinitions [{
      boundBone {boneName=dyn_pepperspray_rt}
      boxExtents=<20 30 30>
      constraintSetup {angularLimitsMax=<5 0 5>, angularLimitsMin=<-5 0 -5>, angularTarget=<1 0 0>}
      localJointOffset=<4 0 0>
     }]
    simSpaceSettings {maxAngularAcceleration=35, maxAngularVelocity=7, simSpaceAngularAlpha=1}
   }
   n8 = Anim Dynamics - Bone: dyn_radio_lf!(ComponentPose=@n7, Alpha=1, GravityOverride:Vector="0,0,-980", ExternalForce="0,0,0") -> (Pose) {
    angularDampingOverride=1
    angularSpringConstant=100
    bAngularSpring=true
    bOverrideAngularDamping=true
    bUseGravityOverride=true
    gravityScale=0.35
    simulationSpace=RootRelative
    boundBone {boneName=dyn_radio_lf}
    componentAppliedLinearAccClamp=<3000 3000 3000>
    componentLinearAccScale=<0.3 0.3 0.3>
    componentLinearVelScale=<0.3 0.3 0.3>
    gravityOverride {z=-980}
    physicsBodyDefinitions [{
      boundBone {boneName=dyn_radio_lf}
      boxExtents=<20 30 30>
      constraintSetup {angularLimitsMax=<5 0 5>, angularLimitsMin=<-3 0 -5>, angularTarget=<1 0 0>}
      localJointOffset=<2 0 0>
     }]
    simSpaceSettings {maxAngularAcceleration=35, maxAngularVelocity=7, simSpaceAngularAlpha=1}
   }
   n9 = Anim Dynamics - Bone: dyn_donut_front_01_lf!(ComponentPose=@n8, Alpha=1, GravityOverride:Vector="0,0,-980", ExternalForce="0,0,0") -> (Pose) {
    angularDampingOverride=1
    angularSpringConstant=50
    bAngularSpring=true
    bOverrideAngularDamping=true
    bUseGravityOverride=true
    gravityScale=0.35
    simulationSpace=RootRelative
    boundBone {boneName=dyn_donut_front_01_lf}
    componentAppliedLinearAccClamp=<3000 3000 3000>
    componentLinearAccScale=<0.2 0.2 0.2>
    componentLinearVelScale=<0.2 0.2 0.2>
    gravityOverride {z=-980}
    physicsBodyDefinitions [{
      boundBone {boneName=dyn_donut_front_01_lf}
      boxExtents=<20 10 20>
      constraintSetup {angularLimitsMax=<3 3 10>, angularLimitsMin=<-3 0 -10>, angularTarget=<1 0 0>}
      localJointOffset=<4 0 0>
     }]
   }
   n10 = Anim Dynamics - Bone: dyn_donut_center_02_lf!(ComponentPose=@n9, Alpha=1, GravityOverride:Vector="0,0,-980", ExternalForce="0,0,0") -> (Pose) {
    angularDampingOverride=1
    angularSpringConstant=80
    bAngularSpring=true
    bOverrideAngularDamping=true
    bUseGravityOverride=true
    gravityScale=0.35
    simulationSpace=RootRelative
    boundBone {boneName=dyn_donut_center_02_lf}
    componentAppliedLinearAccClamp=<3000 3000 3000>
    componentLinearAccScale=<0.2 0.2 0.2>
    componentLinearVelScale=<0.2 0.2 0.2>
    gravityOverride {z=-980}
    physicsBodyDefinitions [{
      boundBone {boneName=dyn_donut_center_02_lf}
      boxExtents=<25 15 25>
      constraintSetup {angularLimitsMax=<5 3 12>, angularLimitsMin=<-5 0 -12>, angularTarget=<1 0 0>}
      localJointOffset=<4 0 0>
     }]
   }
   n11 = Anim Dynamics - Bone: dyn_donut_back_03_lf!(ComponentPose=@n10, Alpha=1, GravityOverride:Vector="0,0,-980", ExternalForce="0,0,0") -> (Pose) {
    angularDampingOverride=1
    angularSpringConstant=50
    bAngularSpring=true
    bOverrideAngularDamping=true
    bUseGravityOverride=true
    gravityScale=0.35
    simulationSpace=RootRelative
    boundBone {boneName=dyn_donut_back_03_lf}
    componentAppliedLinearAccClamp=<3000 3000 3000>
    componentLinearAccScale=<0.2 0.2 0.2>
    componentLinearVelScale=<0.2 0.2 0.2>
    gravityOverride {z=-980}
    physicsBodyDefinitions [{
      boundBone {boneName=dyn_donut_back_03_lf}
      boxExtents=<20 10 20>
      constraintSetup {angularLimitsMax=<-6 3 10>, angularLimitsMin=<-6 0 -10>, angularTarget=<1 0 0>}
      localJointOffset=<4 0 0>
     }]
   }
   Save cached pose 'dyn_cache'(Pose=Component To Local(ComponentPose=@n11)) {cacheName=dyn_cache}
  ]}
 EventGraph {nodes [
   Event BlueprintUpdateAnimation()
   > n1 = Sequence!(execute) -> (then_0, then_1, then_2:Exec)
   > n2 = "Gravity Override State Function (Parameter Asset) (Nativized)"(
    CharacterPartAnimInstance=Self-Reference()
    GameTimeInSeconds=GetGameTimeInSeconds()
    Parameters=&/F_MED_ThinGlaze_Parameters
   )
   > Set GravityOverride(GravityOverride=@n2)
   n3 = Branch(execute=@n1.then_2, Condition=Get bIsInFrontEnd())
   > Set FrontEndValue(FrontEndValue:double=0) -> (Output_Get:double)
   Set FrontEndValue(execute=@n3.else, FrontEndValue=1)
   n4 = Event BlueprintProcessAnimInputEvent()
   n5 = IsNamed(self=@n4.AnimInputEvent, InEventName=aie.Player.Zipline.Entered)
   n6 = OR Boolean(A=Get bIsMoving2D(), B=Get bIsTacticalSprinting())
   n7 = Get bIsJumping()
   n8 = Branch(execute=@n1.then_1, Condition=OR Boolean(A=@n5, B=AND Boolean!(A=@n5, B=@n7, C:bool=@n6) -> (ReturnValue)))
   > Set animDyn on zipline(animDyn on zipline:Vector="0, 0, 0") -> (Output_Get:Vector)
   Branch(execute=@n8.else, Condition=IsNamed(self=@n4.AnimInputEvent, InEventName=aie.Player.Zipline.Exited))
   > Set animDyn on zipline(animDyn on zipline="0.5,0.5,0.5")
  ]}
]
variables [
 animDyn on zipline {type=Vector}
 FrontEndValue {type=number}
 GravityOverride {type=Vector}
]
