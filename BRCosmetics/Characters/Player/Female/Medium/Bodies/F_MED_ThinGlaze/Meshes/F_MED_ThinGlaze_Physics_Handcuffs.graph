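# Physics asset for the handcuffs hanging at the back right of the F_MED_ThinGlaze female character, apparently for secondary motion.
# A base body locked to the pelvis anchors a chain of five dyn_handcuffs bodies, whose swing limits widen toward the tip while twist stays tight or zero.
# Capsules on pelvis, right thigh and calf model the adjacent body, with the thigh joint left free and a larger second capsule on the chain tip.
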PhysicsAsset /BRCosmetics/Characters/Player/Female/Medium/Bodies/F_MED_ThinGlaze/Meshes/F_MED_ThinGlaze_Physics_Handcuffs
bodies {
 base_handcuffs_donut_back_00_rt [{shapeType=Capsule, length=3, radius=1, shapeName=base_handcuffs_donut_back_00_rt_capsule, rotation=<rot 90 0 0>}]
 calf_r [{shapeType=Capsule, length=38.1099, radius=8.7706, center=<23.1108 0.4019 -1.9038>, rotation=<rot 86.5846 -0.0001 -179.9999>}]
 dyn_handcuffs_donut_back_00_rt [{shapeType=Capsule, length=3, radius=1, shapeName=dyn_handcuffs_donut_back_00_rt_capsule, center=<2 0 0.5>, rotation=<rot 90 0 0>}]
 dyn_handcuffs_donut_back_01_rt [{shapeType=Capsule, length=8.9976, radius=2.5309, shapeName=dyn_handcuffs_donut_back_02_rt_capsule, center=<2 0 1.5>, rotation=<rot 0 0 90>}]
 dyn_handcuffs_donut_back_02_rt [{shapeType=Capsule, length=3, radius=1, shapeName=dyn_handcuffs_donut_back_02_rt_capsule, center=<2 0 0>, rotation=<rot 90 0 0>}]
 dyn_handcuffs_donut_back_03_rt [{shapeType=Capsule, length=3, radius=1, shapeName=dyn_handcuffs_donut_back_03_rt_capsule, center=<2 0 0>, rotation=<rot 90 0 0>}]
 dyn_handcuffs_donut_back_04_rt [
  {shapeType=Capsule, length=3, radius=1, shapeName=dyn_handcuffs_donut_back_04_rt_capsule, center=<3 0 0>, rotation=<rot 90 0 0>}
  {
   shapeType=Capsule
   length=10.3446
   radius=2.0934
   shapeName=dyn_handcuffs_donut_back_04_rt_capsule
   center=<15.4855 -2.754 0.1195>
   rotation=<rot -0.0017 -19.0265 -94.9261>
  }
 ]
 pelvis [{shapeType=Capsule, length=16.3957, radius=9.5618, center=<-1.6629 -1.0502 -0.509>}]
 thigh_r [{
   shapeType=Capsule
   length=20.5451
   radius=8.811
   shapeName=thigh_r_capsule
   center=<25.4451 -0.7551 -0.4627>
   rotation=<rot 0 -98.835 -86.7166>
  }]
}
constraints [
 {
  bone1Name=dyn_handcuffs_donut_back_04_rt
  bone2Name=dyn_handcuffs_donut_back_03_rt
  swing1LimitDegrees=30
  swing2LimitDegrees=30
  twistLimitDegrees=8
 }
 {
  bone1Name=dyn_handcuffs_donut_back_03_rt
  bone2Name=dyn_handcuffs_donut_back_02_rt
  swing1LimitDegrees=25
  swing2LimitDegrees=25
  twistLimitDegrees=8
 }
 {
  bone1Name=dyn_handcuffs_donut_back_02_rt
  bone2Name=dyn_handcuffs_donut_back_01_rt
  swing1LimitDegrees=20
  swing2LimitDegrees=20
  twistLimitDegrees=8
 }
 {
  bone1Name=dyn_handcuffs_donut_back_01_rt
  bone2Name=dyn_handcuffs_donut_back_00_rt
  swing1LimitDegrees=15
  swing2LimitDegrees=15
  twistLimitDegrees=0
 }
 {
  bone1Name=dyn_handcuffs_donut_back_00_rt
  bone2Name=base_handcuffs_donut_back_00_rt
  swing1LimitDegrees=15
  swing2LimitDegrees=15
  twistLimitDegrees=0
 }
 {
  bone1Name=base_handcuffs_donut_back_00_rt
  bone2Name=pelvis
  swing1LimitDegrees=45
  swing1Motion=Locked
  swing2LimitDegrees=45
  swing2Motion=Locked
  twistLimitDegrees=45
  twistMotion=Locked
 }
 {
  bone1Name=thigh_r
  bone2Name=pelvis
  swing1LimitDegrees=45
  swing1Motion=Free
  swing2LimitDegrees=45
  swing2Motion=Free
  twistLimitDegrees=45
  twistMotion=Free
 }
 {bone1Name=calf_r, bone2Name=thigh_r, swing1LimitDegrees=45, swing2LimitDegrees=45, twistLimitDegrees=45}
]
